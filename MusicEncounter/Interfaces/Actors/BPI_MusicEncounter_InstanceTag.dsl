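; Blueprint Interface for music encounter actors that lets callers assign an instance tag to them.
; It declares one function, SetMusicEncounterInstanceTag, which takes a gameplay tag identifying the music encounter instance.
; Implementing actors supply the behavior; the interface itself contains no logic.
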(fn SetMusicEncounterInstanceTag
  (params (Music%20Encounter%20Instance%20Tag :type Gameplay%20Tag%20Structure)))
